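# Physics asset for the medium male Llama Hero face accessory, defining dynamics for seven horse-hair bones.
# Each hair bone and the head carries a small capsule body.
# Five hair bodies constrain directly to the head, while 03 and 04 both branch from 02.
# Twist is tightly limited, and swing is mostly wider on one axis than the other, except on 01, which moves most freely.
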PhysicsAsset /Game/Characters/Player/Male/Medium/Bodies/M_MED_Llama_Hero/Meshes/Parts/M_MED_Llama_Hero_FaceAcc_Dyn
bodies {
 dyn_horse_hair_01 [{shapeType=Capsule, length=2, radius=1, center=<3 0 0>, rotation=<rot 90 0 0>}]
 dyn_horse_hair_02 [{shapeType=Capsule, length=2, radius=1, center=<3 0 0>, rotation=<rot 90 0 0>}]
 dyn_horse_hair_03 [{shapeType=Capsule, length=2, radius=1, center=<3 0 0>, rotation=<rot 90 0 0>}]
 dyn_horse_hair_04 [{shapeType=Capsule, length=2, radius=1, center=<3 0 0>, rotation=<rot 90 0 0>}]
 dyn_horse_hair_05 [{shapeType=Capsule, length=2, radius=1, center=<3 0 0>, rotation=<rot 90 0 0>}]
 dyn_horse_hair_06 [{shapeType=Capsule, length=2, radius=1, center=<3 0 0>, rotation=<rot 90 0 0>}]
 dyn_horse_hair_07 [{shapeType=Capsule, length=2, radius=1, center=<3 0 0>, rotation=<rot 90 0 0>}]
 head [{shapeType=Capsule, length=2, radius=1, rotation=<rot 90 0 0>}]
}
constraints [
 {bone1Name=dyn_horse_hair_04, bone2Name=dyn_horse_hair_02, swing1LimitDegrees=20, swing2LimitDegrees=5, twistLimitDegrees=5}
 {bone1Name=dyn_horse_hair_03, bone2Name=dyn_horse_hair_02, swing1LimitDegrees=20, swing2LimitDegrees=5, twistLimitDegrees=5}
 {bone1Name=dyn_horse_hair_01, bone2Name=head, swing1LimitDegrees=30, swing2LimitDegrees=35, twistLimitDegrees=5}
 {bone1Name=dyn_horse_hair_02, bone2Name=head, swing1LimitDegrees=25, swing2LimitDegrees=5, twistLimitDegrees=5}
 {bone1Name=dyn_horse_hair_05, bone2Name=head, swing1LimitDegrees=25, swing2LimitDegrees=5, twistLimitDegrees=5}
 {bone1Name=dyn_horse_hair_06, bone2Name=head, swing1LimitDegrees=25, swing2LimitDegrees=5, twistLimitDegrees=5}
 {bone1Name=dyn_horse_hair_07, bone2Name=head, swing1LimitDegrees=25, swing2LimitDegrees=5, twistLimitDegrees=5}
]
